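# Physics asset for the F_MED_Rover female medium player body, covering only a zipper and the left thigh, apparently for secondary motion of the dangling zipper.
# Small capsule bodies sit on the zipper, zipper_Drvr and thigh_l bones.
# Two constraints chain them: zipper to zipper_Drvr has tight angular limits with a narrow swing on one axis, while zipper_Drvr to thigh_l rotates freely on all axes.
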PhysicsAsset /BRCosmetics/Characters/Player/Female/Medium/Bodies/F_MED_Rover/Meshes/F_MED_Rover_Physics
bodies {
 thigh_l [{shapeType=Capsule, length=1, radius=1, shapeName=thigh_l_capsule, center=<2.1109 0 0>, rotation=<rot 90 0 0>}]
 zipper [{shapeType=Capsule, length=1, radius=1, shapeName=zipper_capsule, center=<3.361 0.0001 0>, rotation=<rot 90 0 0>}]
 zipper_Drvr [{shapeType=Capsule, length=0.505, radius=0.505, shapeName=zipper_Drvr_capsule}]
}
constraints [
 {bone1Name=zipper, bone2Name=zipper_Drvr, swing1LimitDegrees=10, swing2LimitDegrees=0, twistLimitDegrees=0}
 {
  bone1Name=zipper_Drvr
  bone2Name=thigh_l
  swing1LimitDegrees=45
  swing1Motion=Free
  swing2LimitDegrees=45
  swing2Motion=Free
  twistLimitDegrees=45
  twistMotion=Free
 }
]
