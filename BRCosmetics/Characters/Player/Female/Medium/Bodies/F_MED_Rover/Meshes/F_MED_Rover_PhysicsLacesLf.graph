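# Physics asset for the left shoelace of the female medium Rover character body, simulating dangling lace motion.
# It has two small capsule bodies, one on the left calf twist bone and one on the laces_l bone.
# A single constraint joins the laces to the calf, locking twist and one swing axis and allowing only a narrow swing on the other.
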PhysicsAsset /BRCosmetics/Characters/Player/Female/Medium/Bodies/F_MED_Rover/Meshes/F_MED_Rover_PhysicsLacesLf
bodies {
 calf_twist_01_l [{shapeType=Capsule, length=0, radius=2, shapeName=calf_twist_01_l_capsule}]
 laces_l [{shapeType=Capsule, length=1, radius=1, shapeName=laces_l_capsule, center=<3 0 0>, rotation=<rot -90 0 0>}]
}
constraints [{bone1Name=laces_l, bone2Name=calf_twist_01_l, swing1LimitDegrees=0, swing2LimitDegrees=10, twistLimitDegrees=0}]
