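# Post-process style AnimBP for the F_MED_Rover female body cosmetic, layering corrective deformation and secondary motion onto a pose copied from the attached parent mesh.
# Borrowed Textile_Sparkle twist and sculpter Control Rigs feed a parent constraint weighting zipper_Drvr between pelvis and left thigh twist.
# Chained RigidBody sims add body, left/right lace and necklace physics.
# Additive yaws then adjust zipper_Drvr, one scaled by left-thigh rotation against pelvis, computed each update.
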AnimBlueprint /BRCosmetics/Characters/Player/Female/Medium/Bodies/F_MED_Rover/Meshes/F_MED_Rover_AnimBP
target_skeleton=&/F_MED_Rover_Skeleton
graphs [
 AnimGraph {nodes [
   n1 = "Constraint - zipper_Drvr (RR)"!(ComponentPose, Alpha=1, ConstraintWeights_0=0.25, ConstraintWeights_1:float=0.75) -> (Pose) {
    alphaBoolBlend {blendOption=Linear}
    boneToModify {boneName=zipper_Drvr}
    constraintSetup [{transformType=Rotation, targetBone {boneName=pelvis}}, {transformType=Rotation, targetBone {boneName=thigh_l}}]
    constraintWeights [1,1]
   }
   "Constraint - zipper_Drvr (TT)"!(ComponentPose=@n1, Alpha=1, ConstraintWeights_0=0.25, ConstraintWeights_1:float=0.75) -> (Pose) {
    alphaBoolBlend {blendOption=Linear}
    boneToModify {boneName=zipper_Drvr}
    constraintSetup [{transformType=Translation, targetBone {boneName=pelvis}}, {transformType=Translation, targetBone {boneName=thigh_l}}]
    constraintWeights [1,1]
   }
   n2 = Control Rig(Source=Copy Pose From Mesh() {bUseAttachedParent=true}) {
    alphaBoolBlend {blendOption=Linear}
    controlRigAssetReference {blueprintRigClass=/BRCosmetics/Characters/Player/Female/Medium/Bodies/F_MED_Textile_Sparkle/Meshes/F_MED_Textile_Sparkle_Deform_CtrlRig_twists.F_MED_Textile_Sparkle_Deform_CtrlRig_twists_C}
   }
   n3 = Control Rig!(
    Source=@n2
    Alpha
    KneeRot_fr_l_114deg:Vector
    KneeTrans_fr_l_114deg:Vector="0.258,-3.591,0.009"
    KneeScale_fr_l_114deg:Vector="1,1,1"
    KneeScale_fr_l_130deg:Vector="1,1,1"
    KneeRot_fr_l_130deg:Vector="0,0,-51.2"
    KneeTrans_fr_l_130deg:Vector="0.258,-3.591,0.009"
   ) -> (Pose) {
    alphaBoolBlend {blendOption=Linear}
    controlRigAssetReference {blueprintRigClass=/BRCosmetics/Characters/Player/Female/Medium/Bodies/F_MED_Textile_Sparkle/Meshes/F_MED_Textile_Sparkle_Deform_CtrlRig_sculpters.F_MED_Textile_Sparkle_Deform_CtrlRig_sculpters_C}
   }
   n4 = "Constraint - zipper_Drvr (PPPPPPPPPPPPPPPPPPPPPPPPPPPPPPPPPPPP)"!(
    ComponentPose=Local To Component(LocalPose=@n3)
    Alpha=1
    ConstraintWeights_0=0
    ConstraintWeights_1:float=0
    ConstraintWeights_2:float=0
    ConstraintWeights_3:float=0
    ConstraintWeights_4:float=0
    ConstraintWeights_5:float=0
    ConstraintWeights_6:float=0
    ConstraintWeights_7:float=0
    ConstraintWeights_8:float=0
    ConstraintWeights_9:float=0.25
    ConstraintWeights_10:float=0
    ConstraintWeights_11:float=0
    ConstraintWeights_12:float=0
    ConstraintWeights_13:float=0
    ConstraintWeights_14:float=0
    ConstraintWeights_15:float=0
    ConstraintWeights_16:float=0.75
    ConstraintWeights_17:float=0
    ConstraintWeights_18:float=0
    ConstraintWeights_19:float=0
    ConstraintWeights_20:float=0
    ConstraintWeights_21:float=0
    ConstraintWeights_22:float=0
    ConstraintWeights_23:float=0
    ConstraintWeights_24:float=0
    ConstraintWeights_25:float=0
    ConstraintWeights_26:float=0
    ConstraintWeights_27:float=0
    ConstraintWeights_28:float=0
    ConstraintWeights_29:float=0
    ConstraintWeights_30:float=0
    ConstraintWeights_31:float=0
    ConstraintWeights_32:float=0
    ConstraintWeights_33:float=0
    ConstraintWeights_34:float=0
    ConstraintWeights_35:float=0
   ) -> (Pose) {
    alphaBoolBlend {blendOption=Linear}
    boneToModify {boneName=zipper_Drvr}
    constraintSetup [
     {targetBone {boneName=lowerarm_twist_01_l}}
     {targetBone {boneName=lowerarm_l}}
     {targetBone {boneName=lowerarm_twist_02_l}}
     {targetBone {boneName=upperarm_l}}
     {targetBone {boneName=upperarm_twist_02_l}}
     {targetBone {boneName=upperarm_twist_01_l}}
     {targetBone {boneName=spine_05}}
     {targetBone {boneName=clavicle_l}}
     {targetBone {boneName=clavicle_r}}
     {targetBone {boneName=pelvis}}
     {targetBone {boneName=lowerarm_twist_02_r}}
     {targetBone {boneName=lowerarm_twist_01_r}}
     {targetBone {boneName=lowerarm_r}}
     {targetBone {boneName=upperarm_r}}
     {targetBone {boneName=upperarm_twist_02_r}}
     {targetBone {boneName=upperarm_twist_01_r}}
     {targetBone {boneName=thigh_twist_01_l}}
     {targetBone {boneName=calf_r}}
     {targetBone {boneName=calf_twist_01_r}}
     {targetBone {boneName=foot_r}}
     {targetBone {boneName=calf_twist_02_r}}
     {targetBone {boneName=calf_twist_01_l}}
     {targetBone {boneName=foot_l}}
     {targetBone {boneName=neck_01}}
     {targetBone {boneName=head}}
     {targetBone {boneName=neck_02}}
     {targetBone {boneName=deform_elbow_fr_r}}
     {targetBone {boneName=deform_knee_fr_01_r}}
     {targetBone {boneName=deform_elbow_bk_r}}
     {targetBone {boneName=deform_deltoid_fr_r}}
     {targetBone {boneName=deform_elbow_bk_l}}
     {targetBone {boneName=deform_pectoral_fr_l}}
     {targetBone {boneName=deform_deltoid_fr_l}}
     {targetBone {boneName=deform_pectoral_fr_r}}
     {targetBone {boneName=deform_elbow_fr_l}}
     {targetBone {boneName=deform_knee_bk_01_r}}
    ]
    constraintWeights [1,1,1,1,1,1,1,1,1,1,1,1,1,1,1,1,1,1,1,1,1,1,1,1,1,1,1,1,1,1,1,1,1,1,1,1]
   }
   n5 = RigidBody(ComponentPose=@n4) {
    bClampLinearTranslationLimitToRefPose=true
    bOverrideWorldGravity=true
    overridePhysicsAsset=&/F_MED_Rover_Physics
    simulationSpace=BaseBoneSpace
    alphaBoolBlend {blendOption=Linear}
    baseBoneRef {boneName=thigh_l}
    componentAppliedLinearAccClamp=<3000 3000 3000>
    componentLinearAccScale=<0.4 0.3 0.3>
    componentLinearVelScale=<0.4 0.3 0.3>
    overrideWorldGravity {z=-980}
   }
   n6 = RigidBody(ComponentPose=@n5) {
    bClampLinearTranslationLimitToRefPose=true
    bOverrideWorldGravity=true
    overridePhysicsAsset=&/F_MED_Rover_PhysicsLacesLf
    simulationSpace=BaseBoneSpace
    alphaBoolBlend {blendOption=Linear}
    baseBoneRef {boneName=thigh_l}
    componentAppliedLinearAccClamp=<3000 3000 3000>
    componentLinearAccScale=<0.4 0.3 0.3>
    componentLinearVelScale=<0.4 0.3 0.3>
    overrideWorldGravity {z=-980}
   }
   n7 = RigidBody(ComponentPose=@n6) {
    bClampLinearTranslationLimitToRefPose=true
    bOverrideWorldGravity=true
    overridePhysicsAsset=&/F_MED_Rover_PhysicsLacesRt
    simulationSpace=BaseBoneSpace
    alphaBoolBlend {blendOption=Linear}
    baseBoneRef {boneName=thigh_r}
    componentAppliedLinearAccClamp=<3000 3000 3000>
    componentLinearAccScale=<0.4 0.3 0.3>
    componentLinearVelScale=<0.4 0.3 0.3>
    overrideWorldGravity {z=-980}
   }
   n8 = RigidBody(ComponentPose=@n7, Alpha=0.6) {
    bClampLinearTranslationLimitToRefPose=true
    bOverrideWorldGravity=true
    overridePhysicsAsset=&/F_MED_Rover_PhysicsNecklace
    simulationSpace=BaseBoneSpace
    alphaBoolBlend {blendOption=Linear}
    baseBoneRef {boneName=spine_05}
    componentAppliedLinearAccClamp=<3000 3000 3000>
    componentLinearAccScale=<0.4 0.3 0.3>
    componentLinearVelScale=<0.4 0.3 0.3>
    overrideWorldGravity {z=-980}
   }
   n9 = "Transform (Modify) Bone - Bone: zipper_Drvr"(ComponentPose=@n8, Alpha=Get LegLfRot() -> (LegLfRot:double), Translation="0,0,0", Rotation="0,-50,0", Scale="1,1,1") {
    rotationMode=BMM_Additive
    rotationSpace=BCS_BoneSpace
    scaleMode=BMM_Replace
    scaleSpace=BCS_BoneSpace
    translationMode=BMM_Additive
    translationSpace=BCS_BoneSpace
    alphaBoolBlend {blendOption=Linear}
    boneToModify {boneName=zipper_Drvr}
    rotation {yaw=-50}
   }
   n10 = "Transform (Modify) Bone - Bone: zipper_Drvr"(ComponentPose=@n9, Alpha=1, Translation="0,0,0", Rotation="0,-20,0", Scale="1,1,1") {
    rotationMode=BMM_Additive
    rotationSpace=BCS_BoneSpace
    scaleMode=BMM_Replace
    scaleSpace=BCS_BoneSpace
    translationMode=BMM_Additive
    translationSpace=BCS_BoneSpace
    alphaBoolBlend {blendOption=Linear}
    boneToModify {boneName=zipper_Drvr}
    rotation {yaw=-20}
   }
   Output Pose(Result=Component To Local(ComponentPose=@n10)) {layerGroup=None}
  ]}
 EventGraph {nodes [
   Event BlueprintUpdateAnimation()
   > Set LegLfRot(LegLfRot:double=GetAlphaFromReferenceTransformDelta(SkeletalMeshComponent=GetOwnerSkeletalMeshComponent(), BoneName=thigh_l, BaseName=pelvis, InputRange=(X=0,Y=-100), TwistAxis="1,0,0")) -> (Output_Get:double)
  ]}
]
variables [LegLfRot {type=number}]
