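# Physics asset for the F_MED_Rover female character's face accessory mesh, providing bodies and joints for physics-driven hair.
# Seven hair strands (A back/left/right, B and C left/right) are capsule chains, each running from a HOLD bone constrained to the head through four segments.
# Capsules on head, neck, upper spine, clavicles and upper arms form the body structure.
# Extra constraints tie the tips of two neighboring strand pairs together (B–C right, A–B left).
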PhysicsAsset /BRCosmetics/Characters/Player/Female/Medium/Bodies/F_MED_Rover/Meshes/Parts/F_MED_Rover_FaceAcc_Physics
bodies {
 HairA1_bk [{shapeType=Capsule, length=1, radius=2.5, shapeName=HairA1_bk_capsule, center=<0 0 1.4>, rotation=<rot 90 0 0>}]
 HairA1_l [{shapeType=Capsule, length=1, radius=2.5, shapeName=HairA1_l_capsule, center=<0.5 0 0.75>, rotation=<rot 90 0 0>}]
 HairA1_r [{shapeType=Capsule, length=1, radius=2.5, shapeName=HairA1_r_capsule, center=<0 0 1.7>, rotation=<rot 90 0 0>}]
 HairA2_bk [{shapeType=Capsule, length=1, radius=2.5, shapeName=HairA2_bk_capsule, center=<0 0 1.4>, rotation=<rot 90 0 0>}]
 HairA2_l [{shapeType=Capsule, length=1, radius=2.5, shapeName=HairA2_l_capsule, center=<0.5 0 0.75>, rotation=<rot 90 0 0>}]
 HairA2_r [{shapeType=Capsule, length=5, radius=2.5, shapeName=HairA2_r_capsule, center=<0 0 1.7>, rotation=<rot 90 0 0>}]
 HairA3_bk [{shapeType=Capsule, length=1, radius=2.5, shapeName=HairA3_bk_capsule, center=<0 0 1.4>, rotation=<rot 90 0 0>}]
 HairA3_l [{shapeType=Capsule, length=1, radius=2.5, shapeName=HairA3_l_capsule, center=<0.5 0 0.75>, rotation=<rot 90 0 0>}]
 HairA3_r [{shapeType=Capsule, length=3.5, radius=2.5, shapeName=HairA3_r_capsule, center=<0 0 1.7>, rotation=<rot 90 0 0>}]
 HairA4_bk [{shapeType=Capsule, length=1.25, radius=2.25, shapeName=HairA4_bk_capsule, center=<0 0 1.4>, rotation=<rot 90 0 0>}]
 HairA4_l [{shapeType=Capsule, length=1, radius=2.5, shapeName=HairA4_l_capsule, center=<0.5 0 0.75>, rotation=<rot 90 0 0>}]
 HairA4_r [{shapeType=Capsule, length=2, radius=2.5, shapeName=HairA4_r_capsule, center=<0.5 0 1.7>, rotation=<rot 90 0 0>}]
 HairA_HOLD_bk [{shapeType=Capsule, length=1, radius=2.5, shapeName=HairA_HOLD_bk_capsule, rotation=<rot 90 0 0>}]
 HairA_HOLD_l [{shapeType=Capsule, length=1, radius=2.5, shapeName=HairA_HOLD_l_capsule, rotation=<rot 90 0 0>}]
 HairA_HOLD_r [{shapeType=Capsule, length=1, radius=2.5, shapeName=HairA_HOLD_r_capsule, rotation=<rot 90 0 0>}]
 HairB1_l [{shapeType=Capsule, length=1, radius=2.5, shapeName=HairB1_l_capsule, center=<0.5 0 1.7>, rotation=<rot 90 0 0>}]
 HairB1_r [{shapeType=Capsule, length=1, radius=2.5, shapeName=HairB1_r_capsule, center=<0.5 0 1.8>, rotation=<rot 90 0 0>}]
 HairB2_l [{shapeType=Capsule, length=1, radius=2.5, shapeName=HairB2_l_capsule, center=<0.5 0 1.7>, rotation=<rot 90 0 0>}]
 HairB2_r [{shapeType=Capsule, length=5, radius=2.5, shapeName=HairB2_r_capsule, center=<0.5 0 1.8>, rotation=<rot 90 0 0>}]
 HairB3_l [{shapeType=Capsule, length=1, radius=2.5, shapeName=HairB3_l_capsule, center=<0.5 0 1.7>, rotation=<rot 90 0 0>}]
 HairB3_r [{shapeType=Capsule, length=3.5, radius=2.5, shapeName=HairB3_r_capsule, center=<0.5 0 1.8>, rotation=<rot 90 0 0>}]
 HairB4_l [{shapeType=Capsule, length=0.75, radius=2.25, shapeName=HairB4_l_capsule, center=<0.5 0 1.7>, rotation=<rot 90 0 0>}]
 HairB4_r [{shapeType=Capsule, length=0.25, radius=2.5, shapeName=HairB3_r_capsule, center=<0.1 0 1.8>, rotation=<rot 90 0.6364 0.6414>}]
 HairB_HOLD_l [{shapeType=Capsule, length=1, radius=2.5, shapeName=HairB_HOLD_l_capsule, center=<0.5 0 0>, rotation=<rot 90 0 0>}]
 HairB_HOLD_r [{shapeType=Capsule, length=1, radius=2.5, shapeName=HairB_HOLD_r_capsule, rotation=<rot 90 0 0>}]
 HairC1_l [{shapeType=Capsule, length=1.25, radius=2.5, shapeName=HairC1_l_capsule, center=<1 0 0.95>, rotation=<rot 90 0 0>}]
 HairC1_r [{shapeType=Capsule, length=1, radius=2.5, shapeName=HairC1_r_capsule, center=<0.5 2 1>, rotation=<rot 90 0 0>}]
 HairC2_l [{shapeType=Capsule, length=1.25, radius=2.5, shapeName=HairC2_l_capsule, center=<1 0 0.95>, rotation=<rot 90 0 0>}]
 HairC2_r [{shapeType=Capsule, length=5, radius=2.5, shapeName=HairC2_r_capsule, center=<0.5 2 1>, rotation=<rot 90 0 0>}]
 HairC3_l [{shapeType=Capsule, length=1.25, radius=2.5, shapeName=HairC3_l_capsule, center=<1 0 0.95>, rotation=<rot 90 0 0>}]
 HairC3_r [{shapeType=Capsule, length=3.5, radius=2.5, shapeName=HairC3_r_capsule, center=<0.5 2 1>, rotation=<rot 90 0 0>}]
 HairC4_l [{shapeType=Capsule, length=1, radius=2.25, shapeName=HairC4_l_capsule, center=<1 0 0.95>, rotation=<rot 90 0 0>}]
 HairC4_r [{shapeType=Capsule, length=1, radius=2.5, shapeName=HairC4_r_capsule, center=<0.5 2 1>, rotation=<rot 90 0 0>}]
 HairC_HOLD_l [{shapeType=Capsule, length=1, radius=2.5, shapeName=HairC_HOLD_l_capsule, rotation=<rot 90 0 0>}]
 HairC_HOLD_r [{shapeType=Capsule, length=1, radius=2.5, shapeName=HairC_HOLD_r_capsule, rotation=<rot 90 0 0>}]
 clavicle_l [
  {
   shapeType=Capsule
   length=2.25
   radius=8.255
   shapeName=clavicle_l_capsule
   center=<5.0002 -0.7464 -2.3215>
   rotation=<rot 0 -89.9996 99.9991>
  }
  {shapeType=Capsule, length=5.5, radius=8.25, shapeName=clavicle_l_capsule, center=<5.3243 9.1067 -4.6956>, rotation=<rot -90 -1.2659 1.266>}
 ]
 clavicle_r [{shapeType=Capsule, length=11.5, radius=6.25, shapeName=clavicle_r_capsule, center=<-0.7321 0.0002 0.064>, rotation=<rot -90 -180 0>}]
 head [{shapeType=Capsule, length=4.5, radius=6.5, shapeName=head_capsule, center=<3.5564 -0.0001 0>, rotation=<rot -90 1.2658 -1.2658>}]
 neck_01 [{shapeType=Capsule, length=0.5, radius=5.75, shapeName=neck_01_capsule, rotation=<rot -90 -1.8984 1.8984>}]
 neck_02 [{shapeType=Capsule, length=0.75, radius=6, shapeName=neck_02_capsule, center=<-0.0004 1.4079 0>, rotation=<rot -90 -0.2373 0.2375>}]
 spine_05 [{shapeType=Capsule, length=8.25, radius=10.5, shapeName=spine_05_capsule, center=<0.7791 -0.2665 -0.0001>}]
 upperarm_l [{
   shapeType=Capsule
   length=19.5
   radius=5.505
   shapeName=upperarm_l_capsule
   center=<11.2153 -0.0002 -0.0002>
   rotation=<rot 0 -89.9999 -90.0011>
  }]
 upperarm_r [{
   shapeType=Capsule
   length=19.5
   radius=5.505
   shapeName=upperarm_r_capsule
   center=<-11.2153 -0.0002 -0.0002>
   rotation=<rot 0 -89.9999 -90.0011>
  }]
}
constraints [
 {
  bone1Name=head
  bone2Name=neck_02
  swing1LimitDegrees=45
  swing1Motion=Free
  swing2LimitDegrees=45
  swing2Motion=Free
  twistLimitDegrees=45
  twistMotion=Free
 }
 {
  bone1Name=neck_02
  bone2Name=neck_01
  swing1LimitDegrees=45
  swing1Motion=Free
  swing2LimitDegrees=45
  swing2Motion=Free
  twistLimitDegrees=45
  twistMotion=Free
 }
 {
  bone1Name=neck_01
  bone2Name=spine_05
  swing1LimitDegrees=45
  swing1Motion=Free
  swing2LimitDegrees=45
  swing2Motion=Free
  twistLimitDegrees=45
  twistMotion=Free
 }
 {
  bone1Name=clavicle_l
  bone2Name=spine_05
  swing1LimitDegrees=45
  swing1Motion=Free
  swing2LimitDegrees=45
  swing2Motion=Free
  twistLimitDegrees=45
  twistMotion=Free
 }
 {
  bone1Name=clavicle_r
  bone2Name=spine_05
  swing1LimitDegrees=45
  swing1Motion=Free
  swing2LimitDegrees=45
  swing2Motion=Free
  twistLimitDegrees=45
  twistMotion=Free
 }
 {bone1Name=HairC_HOLD_l, bone2Name=head, swing1LimitDegrees=45, swing2LimitDegrees=45, twistLimitDegrees=0}
 {
  bone1Name=HairC1_l
  bone2Name=HairC_HOLD_l
  swing1LimitDegrees=60
  swing1Motion=Free
  swing2LimitDegrees=60
  swing2Motion=Free
  twistLimitDegrees=0
 }
 {
  bone1Name=HairC2_l
  bone2Name=HairC1_l
  swing1LimitDegrees=60
  swing1Motion=Free
  swing2LimitDegrees=60
  swing2Motion=Free
  twistLimitDegrees=0
 }
 {
  bone1Name=HairC3_l
  bone2Name=HairC2_l
  swing1LimitDegrees=60
  swing1Motion=Free
  swing2LimitDegrees=60
  swing2Motion=Free
  twistLimitDegrees=0
 }
 {
  bone1Name=HairC4_l
  bone2Name=HairC3_l
  swing1LimitDegrees=20
  swing1Motion=Free
  swing2LimitDegrees=20
  swing2Motion=Free
  twistLimitDegrees=0
 }
 {
  bone1Name=HairB_HOLD_l
  bone2Name=head
  swing1LimitDegrees=45
  swing1Motion=Free
  swing2LimitDegrees=45
  swing2Motion=Free
  twistLimitDegrees=0
 }
 {
  bone1Name=HairB1_l
  bone2Name=HairB_HOLD_l
  swing1LimitDegrees=60
  swing1Motion=Free
  swing2LimitDegrees=60
  swing2Motion=Free
  twistLimitDegrees=0
 }
 {
  bone1Name=HairB2_l
  bone2Name=HairB1_l
  swing1LimitDegrees=60
  swing1Motion=Free
  swing2LimitDegrees=60
  swing2Motion=Free
  twistLimitDegrees=0
 }
 {
  bone1Name=HairB3_l
  bone2Name=HairB2_l
  swing1LimitDegrees=60
  swing1Motion=Free
  swing2LimitDegrees=60
  swing2Motion=Free
  twistLimitDegrees=0
 }
 {
  bone1Name=HairB4_l
  bone2Name=HairB3_l
  swing1LimitDegrees=20
  swing1Motion=Free
  swing2LimitDegrees=20
  swing2Motion=Free
  twistLimitDegrees=0
 }
 {
  bone1Name=HairA_HOLD_bk
  bone2Name=head
  swing1LimitDegrees=45
  swing1Motion=Free
  swing2LimitDegrees=45
  swing2Motion=Free
  twistLimitDegrees=0
 }
 {
  bone1Name=HairA1_bk
  bone2Name=HairA_HOLD_bk
  swing1LimitDegrees=60
  swing1Motion=Free
  swing2LimitDegrees=60
  swing2Motion=Free
  twistLimitDegrees=0
 }
 {
  bone1Name=HairA2_bk
  bone2Name=HairA1_bk
  swing1LimitDegrees=60
  swing1Motion=Free
  swing2LimitDegrees=60
  swing2Motion=Free
  twistLimitDegrees=0
 }
 {
  bone1Name=HairA3_bk
  bone2Name=HairA2_bk
  swing1LimitDegrees=60
  swing1Motion=Free
  swing2LimitDegrees=60
  swing2Motion=Free
  twistLimitDegrees=0
 }
 {
  bone1Name=HairA4_bk
  bone2Name=HairA3_bk
  swing1LimitDegrees=20
  swing1Motion=Free
  swing2LimitDegrees=20
  swing2Motion=Free
  twistLimitDegrees=0
 }
 {
  bone1Name=HairA_HOLD_l
  bone2Name=head
  swing1LimitDegrees=45
  swing1Motion=Free
  swing2LimitDegrees=45
  swing2Motion=Free
  twistLimitDegrees=0
 }
 {
  bone1Name=HairA1_l
  bone2Name=HairA_HOLD_l
  swing1LimitDegrees=60
  swing1Motion=Free
  swing2LimitDegrees=60
  swing2Motion=Free
  twistLimitDegrees=0
 }
 {
  bone1Name=HairA2_l
  bone2Name=HairA1_l
  swing1LimitDegrees=60
  swing1Motion=Free
  swing2LimitDegrees=60
  swing2Motion=Free
  twistLimitDegrees=0
 }
 {
  bone1Name=HairA3_l
  bone2Name=HairA2_l
  swing1LimitDegrees=60
  swing1Motion=Free
  swing2LimitDegrees=60
  swing2Motion=Free
  twistLimitDegrees=0
 }
 {
  bone1Name=HairA4_l
  bone2Name=HairA3_l
  swing1LimitDegrees=20
  swing1Motion=Free
  swing2LimitDegrees=20
  swing2Motion=Free
  twistLimitDegrees=0
 }
 {
  bone1Name=HairA_HOLD_r
  bone2Name=head
  swing1LimitDegrees=45
  swing1Motion=Free
  swing2LimitDegrees=45
  swing2Motion=Free
  twistLimitDegrees=0
 }
 {
  bone1Name=HairA1_r
  bone2Name=HairA_HOLD_r
  swing1LimitDegrees=60
  swing1Motion=Free
  swing2LimitDegrees=60
  swing2Motion=Free
  twistLimitDegrees=0
 }
 {
  bone1Name=HairA2_r
  bone2Name=HairA1_r
  swing1LimitDegrees=60
  swing1Motion=Free
  swing2LimitDegrees=60
  swing2Motion=Free
  twistLimitDegrees=0
 }
 {
  bone1Name=HairA3_r
  bone2Name=HairA2_r
  swing1LimitDegrees=60
  swing1Motion=Free
  swing2LimitDegrees=60
  swing2Motion=Free
  twistLimitDegrees=0
 }
 {
  bone1Name=HairA4_r
  bone2Name=HairA3_r
  swing1LimitDegrees=20
  swing1Motion=Free
  swing2LimitDegrees=20
  swing2Motion=Free
  twistLimitDegrees=0
 }
 {
  bone1Name=HairC_HOLD_r
  bone2Name=head
  swing1LimitDegrees=45
  swing1Motion=Free
  swing2LimitDegrees=45
  swing2Motion=Free
  twistLimitDegrees=0
 }
 {
  bone1Name=HairC1_r
  bone2Name=HairC_HOLD_r
  swing1LimitDegrees=60
  swing1Motion=Free
  swing2LimitDegrees=60
  swing2Motion=Free
  twistLimitDegrees=0
 }
 {
  bone1Name=HairC2_r
  bone2Name=HairC1_r
  swing1LimitDegrees=60
  swing1Motion=Free
  swing2LimitDegrees=60
  swing2Motion=Free
  twistLimitDegrees=0
 }
 {
  bone1Name=HairC3_r
  bone2Name=HairC2_r
  swing1LimitDegrees=60
  swing1Motion=Free
  swing2LimitDegrees=60
  swing2Motion=Free
  twistLimitDegrees=0
 }
 {
  bone1Name=HairC4_r
  bone2Name=HairC3_r
  swing1LimitDegrees=20
  swing1Motion=Free
  swing2LimitDegrees=20
  swing2Motion=Free
  twistLimitDegrees=0
 }
 {
  bone1Name=HairB_HOLD_r
  bone2Name=head
  swing1LimitDegrees=45
  swing1Motion=Free
  swing2LimitDegrees=45
  swing2Motion=Free
  twistLimitDegrees=0
 }
 {
  bone1Name=HairB1_r
  bone2Name=HairB_HOLD_r
  swing1LimitDegrees=90
  swing1Motion=Free
  swing2LimitDegrees=90
  swing2Motion=Free
  twistLimitDegrees=0
 }
 {bone1Name=HairB2_r, bone2Name=HairB1_r, swing1LimitDegrees=90, swing2LimitDegrees=90, twistLimitDegrees=0}
 {bone1Name=HairB3_r, bone2Name=HairB2_r, swing1LimitDegrees=90, swing2LimitDegrees=90, twistLimitDegrees=0}
 {
  bone1Name=upperarm_l
  bone2Name=clavicle_l
  swing1LimitDegrees=45
  swing1Motion=Free
  swing2LimitDegrees=45
  swing2Motion=Free
  twistLimitDegrees=45
  twistMotion=Free
 }
 {
  bone1Name=upperarm_r
  bone2Name=clavicle_r
  swing1LimitDegrees=45
  swing1Motion=Free
  swing2LimitDegrees=45
  swing2Motion=Free
  twistLimitDegrees=45
  twistMotion=Free
 }
 {bone1Name=HairB4_r, bone2Name=HairB3_r, swing1LimitDegrees=90, swing2LimitDegrees=90, twistLimitDegrees=0}
 {
  bone1Name=HairB4_r
  bone2Name=HairC4_r
  swing1LimitDegrees=45
  swing1Motion=Free
  swing2LimitDegrees=45
  swing2Motion=Free
  twistLimitDegrees=45
  twistMotion=Free
 }
 {
  bone1Name=HairA4_l
  bone2Name=HairB4_l
  swing1LimitDegrees=45
  swing1Motion=Free
  swing2LimitDegrees=45
  swing2Motion=Free
  twistLimitDegrees=45
  twistMotion=Free
 }
]
