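# Anim Blueprint for the Rover female FaceAcc mesh that poses and simulates its hair bones on the player character.
# A cached pose copied from parent and body meshes feeds bone offsets and RigidBody sims with separate normal, DanceMoves emote and skydive physics assets.
# EventGraph derives their alphas from head-to-spine tilt, movement, frontend and emote state.
# DBNO switches to a simpler chain: one sim plus fixed hair rotations.
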AnimBlueprint /BRCosmetics/Characters/Player/Female/Medium/Bodies/F_MED_Rover/Meshes/Parts/F_MED_Rover_FaceAcc_AnimBP
target_skeleton=&/F_MED_Rover_FaceAcc_Skeleton
graphs [
 AnimGraph {nodes [
   Local To Component(LocalPose=Copy Pose From Mesh() {bUseAttachedParent=true})
   n1 = Look At - Bone: HairC_POSING_l(Alpha=1) {
    alphaBoolBlend {blendOption=Linear}
    boneToModify {boneName=HairC_POSING_l}
    lookAtLocation=<-1.2131 -21.0413 -5.7686>
    lookAtTarget {boneReference {boneName=spine_05}}
    lookAt_Axis {axis=<1 0>}
   }
   n2 = Look At - Bone: HairA_POSING_bk(ComponentPose=@n1, Alpha=1) {
    alphaBoolBlend {blendOption=Linear}
    boneToModify {boneName=HairA_POSING_bk}
    lookAtLocation=<4.1279 -12.5947 5.0252>
    lookAtTarget {boneReference {boneName=spine_05}}
    lookAt_Axis {axis=<1 0>}
   }
   Look At - Bone: HairC_POSING_r(ComponentPose=@n2, Alpha=1) {
    alphaBoolBlend {blendOption=Linear}
    boneToModify {boneName=HairC_POSING_r}
    lookAtLocation=<0.0005 -15.3827 12.5994>
    lookAtTarget {boneReference {boneName=spine_05}}
    lookAt_Axis {axis=<1 0>}
   }
   n3 = "Transform (Modify) Bone - Bone: HairA3_bk"(Alpha=1, Translation="0,0,-7", Rotation="0,0,0", Scale="1,1,1") {
    rotationMode=BMM_Additive
    rotationSpace=BCS_BoneSpace
    scaleMode=BMM_Replace
    scaleSpace=BCS_BoneSpace
    translationMode=BMM_Additive
    translationSpace=BCS_BoneSpace
    alphaBoolBlend {blendOption=Linear}
    alphaScaleBiasClamp {bInterpResult=true, interpSpeedIncreasing=1}
    boneToModify {boneName=HairA3_bk}
    translation {z=-7}
   }
   n4 = "Transform (Modify) Bone - Bone: HairC3_r"(ComponentPose=@n3, Alpha=1, Translation="0,0,-5", Rotation="0,0,0", Scale="1,1,1") {
    rotationMode=BMM_Additive
    rotationSpace=BCS_BoneSpace
    scaleMode=BMM_Replace
    scaleSpace=BCS_BoneSpace
    translationMode=BMM_Additive
    translationSpace=BCS_BoneSpace
    alphaBoolBlend {blendOption=Linear}
    alphaScaleBiasClamp {bInterpResult=true, interpSpeedIncreasing=1}
    boneToModify {boneName=HairC3_r}
    translation {z=-5}
   }
   n5 = "Transform (Modify) Bone - Bone: HairC3_l"(ComponentPose=@n4, Alpha=1, Translation="0,0,-4", Rotation="0,0,0", Scale="1,1,1") {
    rotationMode=BMM_Additive
    rotationSpace=BCS_BoneSpace
    scaleMode=BMM_Replace
    scaleSpace=BCS_BoneSpace
    translationMode=BMM_Additive
    translationSpace=BCS_BoneSpace
    alphaBoolBlend {blendOption=Linear}
    alphaScaleBiasClamp {bInterpResult=true, interpSpeedIncreasing=1}
    boneToModify {boneName=HairC3_l}
    translation {z=-4}
   }
   n6 = "Transform (Modify) Bone - Bone: HairA_POSING_l"(ComponentPose=@n5, Alpha=1, Translation="-4,0,0", Rotation="0,0,0", Scale="1,1,1") {
    rotationMode=BMM_Additive
    rotationSpace=BCS_BoneSpace
    scaleMode=BMM_Replace
    scaleSpace=BCS_BoneSpace
    translationMode=BMM_Additive
    translationSpace=BCS_BoneSpace
    alphaBoolBlend {blendOption=Linear}
    alphaScaleBiasClamp {bInterpResult=true, interpSpeedIncreasing=1}
    boneToModify {boneName=HairA_POSING_l}
    translation {x=-4}
   }
   n7 = "Transform (Modify) Bone - Bone: HairB_POSING_l"(ComponentPose=@n6, Alpha=1, Translation="-4,0,0", Rotation="0,0,0", Scale="1,1,1") {
    rotationMode=BMM_Additive
    rotationSpace=BCS_BoneSpace
    scaleMode=BMM_Replace
    scaleSpace=BCS_BoneSpace
    translationMode=BMM_Additive
    translationSpace=BCS_BoneSpace
    alphaBoolBlend {blendOption=Linear}
    alphaScaleBiasClamp {bInterpResult=true, interpSpeedIncreasing=1}
    boneToModify {boneName=HairB_POSING_l}
    translation {x=-4}
   }
   n8 = "Transform (Modify) Bone - Bone: HairC_POSING_l"(ComponentPose=@n7, Alpha=1, Translation="-4,0,0", Rotation="0,0,0", Scale="1,1,1") {
    rotationMode=BMM_Additive
    rotationSpace=BCS_BoneSpace
    scaleMode=BMM_Replace
    scaleSpace=BCS_BoneSpace
    translationMode=BMM_Additive
    translationSpace=BCS_BoneSpace
    alphaBoolBlend {blendOption=Linear}
    alphaScaleBiasClamp {bInterpResult=true, interpSpeedIncreasing=1}
    boneToModify {boneName=HairC_POSING_l}
    translation {x=-4}
   }
   n9 = "Transform (Modify) Bone - Bone: HairA_POSING_bk"(ComponentPose=@n8, Alpha=1, Translation="-3,0,0", Rotation="0,0,0", Scale="1,1,1") {
    rotationMode=BMM_Additive
    rotationSpace=BCS_BoneSpace
    scaleMode=BMM_Replace
    scaleSpace=BCS_BoneSpace
    translationMode=BMM_Additive
    translationSpace=BCS_BoneSpace
    alphaBoolBlend {blendOption=Linear}
    alphaScaleBiasClamp {bInterpResult=true, interpSpeedIncreasing=1}
    boneToModify {boneName=HairA_POSING_bk}
    translation {x=-3}
   }
   n10 = "Transform (Modify) Bone - Bone: HairC_POSING_r"(ComponentPose=@n9, Alpha=1, Translation="-4,0,0", Rotation="0,0,0", Scale="1,1,1") {
    rotationMode=BMM_Additive
    rotationSpace=BCS_BoneSpace
    scaleMode=BMM_Replace
    scaleSpace=BCS_BoneSpace
    translationMode=BMM_Additive
    translationSpace=BCS_BoneSpace
    alphaBoolBlend {blendOption=Linear}
    alphaScaleBiasClamp {bInterpResult=true, interpSpeedIncreasing=1}
    boneToModify {boneName=HairC_POSING_r}
    translation {x=-4}
   }
   "Transform (Modify) Bone - Bone: HairB_POSING_r"(ComponentPose=@n10, Alpha=1, Translation="-4,0,0", Rotation="0,0,0", Scale="1,1,1") {
    rotationMode=BMM_Additive
    rotationSpace=BCS_BoneSpace
    scaleMode=BMM_Replace
    scaleSpace=BCS_BoneSpace
    translationMode=BMM_Additive
    translationSpace=BCS_BoneSpace
    alphaBoolBlend {blendOption=Linear}
    alphaScaleBiasClamp {bInterpResult=true, interpSpeedIncreasing=1}
    boneToModify {boneName=HairB_POSING_r}
    translation {x=-4}
   }
   n11 = "Transform (Modify) Bone - Bone: HairC_POSING_l"(Alpha=1, Translation="0,0,0", Rotation="45,-45,0", Scale="1,1,1") {
    rotationMode=BMM_Additive
    rotationSpace=BCS_BoneSpace
    scaleMode=BMM_Replace
    scaleSpace=BCS_BoneSpace
    translationMode=BMM_Additive
    translationSpace=BCS_BoneSpace
    alphaBoolBlend {blendOption=Linear}
    boneToModify {boneName=HairC_POSING_l}
    rotation {pitch=45, yaw=-45}
   }
   n12 = "Transform (Modify) Bone - Bone: HairB_POSING_l"(ComponentPose=@n11, Alpha=1, Translation="0,0,0", Rotation="45,-45,0", Scale="1,1,1") {
    rotationMode=BMM_Additive
    rotationSpace=BCS_BoneSpace
    scaleMode=BMM_Replace
    scaleSpace=BCS_BoneSpace
    translationMode=BMM_Additive
    translationSpace=BCS_BoneSpace
    alphaBoolBlend {blendOption=Linear}
    boneToModify {boneName=HairB_POSING_l}
    rotation {pitch=45, yaw=-45}
   }
   n13 = "Transform (Modify) Bone - Bone: HairA_POSING_bk"(ComponentPose=@n12, Alpha=1, Translation="0,0,0", Rotation="45,0,0", Scale="1,1,1") {
    rotationMode=BMM_Additive
    rotationSpace=BCS_BoneSpace
    scaleMode=BMM_Replace
    scaleSpace=BCS_BoneSpace
    translationMode=BMM_Additive
    translationSpace=BCS_BoneSpace
    alphaBoolBlend {blendOption=Linear}
    boneToModify {boneName=HairA_POSING_bk}
    rotation {pitch=45}
   }
   n14 = "Transform (Modify) Bone - Bone: HairA_POSING_l"(ComponentPose=@n13, Alpha=1, Translation="0,0,0", Rotation="45,-45,0", Scale="1,1,1") {
    rotationMode=BMM_Additive
    rotationSpace=BCS_BoneSpace
    scaleMode=BMM_Replace
    scaleSpace=BCS_BoneSpace
    translationMode=BMM_Additive
    translationSpace=BCS_BoneSpace
    alphaBoolBlend {blendOption=Linear}
    boneToModify {boneName=HairA_POSING_l}
    rotation {pitch=45, yaw=-45}
   }
   n15 = "Transform (Modify) Bone - Bone: HairA_POSING_r"(ComponentPose=@n14, Alpha=1, Translation="0,0,0", Rotation="45,30,0", Scale="1,1,1") {
    rotationMode=BMM_Additive
    rotationSpace=BCS_BoneSpace
    scaleMode=BMM_Replace
    scaleSpace=BCS_BoneSpace
    translationMode=BMM_Additive
    translationSpace=BCS_BoneSpace
    alphaBoolBlend {blendOption=Linear}
    boneToModify {boneName=HairA_POSING_r}
    rotation {pitch=45, yaw=30}
   }
   n16 = "Transform (Modify) Bone - Bone: HairC_POSING_r"(ComponentPose=@n15, Alpha=1, Translation="0,0,0", Rotation="45,0,0", Scale="1,1,1") {
    rotationMode=BMM_Additive
    rotationSpace=BCS_BoneSpace
    scaleMode=BMM_Replace
    scaleSpace=BCS_BoneSpace
    translationMode=BMM_Additive
    translationSpace=BCS_BoneSpace
    alphaBoolBlend {blendOption=Linear}
    boneToModify {boneName=HairC_POSING_r}
    rotation {pitch=45}
   }
   "Transform (Modify) Bone - Bone: HairB_POSING_r"(ComponentPose=@n16, Alpha=1, Translation="0,0,0", Rotation="45,-20,0", Scale="1,1,1") {
    rotationMode=BMM_Additive
    rotationSpace=BCS_BoneSpace
    scaleMode=BMM_Replace
    scaleSpace=BCS_BoneSpace
    translationMode=BMM_Additive
    translationSpace=BCS_BoneSpace
    alphaBoolBlend {blendOption=Linear}
    boneToModify {boneName=HairB_POSING_r}
    rotation {pitch=45, yaw=-20}
   }
   n17 = Get CusomEmoteFrontend() -> (CusomEmoteFrontend:double)
   n18 = "Transform (Modify) Bone - Bone: HairC_POSING_l"(Alpha=@n17, Translation="0,0,0", Rotation="0,0,0", Scale="0.9,1,1") {
    rotationMode=BMM_Additive
    rotationSpace=BCS_BoneSpace
    scaleMode=BMM_Replace
    scaleSpace=BCS_BoneSpace
    translationMode=BMM_Additive
    translationSpace=BCS_BoneSpace
    alphaBoolBlend {blendOption=Linear}
    alphaScaleBiasClamp {interpSpeedDecreasing=1, interpSpeedIncreasing=1}
    boneToModify {boneName=HairC_POSING_l}
    scale {x=0.9}
   }
   n19 = "Transform (Modify) Bone - Bone: HairB_POSING_l"(ComponentPose=@n18, Alpha=@n17, Translation="0,0,0", Rotation="0,0,0", Scale="0.9,1,1") {
    rotationMode=BMM_Additive
    rotationSpace=BCS_BoneSpace
    scaleMode=BMM_Replace
    scaleSpace=BCS_BoneSpace
    translationMode=BMM_Additive
    translationSpace=BCS_BoneSpace
    alphaBoolBlend {blendOption=Linear}
    alphaScaleBiasClamp {interpSpeedDecreasing=1, interpSpeedIncreasing=1}
    boneToModify {boneName=HairB_POSING_l}
    scale {x=0.9}
   }
   n20 = "Transform (Modify) Bone - Bone: HairA_POSING_bk"(ComponentPose=@n19, Alpha=@n17, Translation="0,0,0", Rotation="0,0,0", Scale="0.9,1,1") {
    rotationMode=BMM_Additive
    rotationSpace=BCS_BoneSpace
    scaleMode=BMM_Replace
    scaleSpace=BCS_BoneSpace
    translationMode=BMM_Additive
    translationSpace=BCS_BoneSpace
    alphaBoolBlend {blendOption=Linear}
    alphaScaleBiasClamp {interpSpeedDecreasing=1, interpSpeedIncreasing=1}
    boneToModify {boneName=HairA_POSING_bk}
    scale {x=0.9}
   }
   n21 = "Transform (Modify) Bone - Bone: HairA_POSING_l"(ComponentPose=@n20, Alpha=@n17, Translation="0,0,0", Rotation="0,0,0", Scale="0.9,1,1") {
    rotationMode=BMM_Additive
    rotationSpace=BCS_BoneSpace
    scaleMode=BMM_Replace
    scaleSpace=BCS_BoneSpace
    translationMode=BMM_Additive
    translationSpace=BCS_BoneSpace
    alphaBoolBlend {blendOption=Linear}
    alphaScaleBiasClamp {interpSpeedDecreasing=1, interpSpeedIncreasing=1}
    boneToModify {boneName=HairA_POSING_l}
    scale {x=0.9}
   }
   n22 = "Transform (Modify) Bone - Bone: HairA_POSING_r"(ComponentPose=@n21, Alpha=@n17, Translation="0,0,0", Rotation="0,0,0", Scale="0.9,1,1") {
    rotationMode=BMM_Additive
    rotationSpace=BCS_BoneSpace
    scaleMode=BMM_Replace
    scaleSpace=BCS_BoneSpace
    translationMode=BMM_Additive
    translationSpace=BCS_BoneSpace
    alphaBoolBlend {blendOption=Linear}
    alphaScaleBiasClamp {interpSpeedDecreasing=1, interpSpeedIncreasing=1}
    boneToModify {boneName=HairA_POSING_r}
    scale {x=0.9}
   }
   n23 = "Transform (Modify) Bone - Bone: HairC_POSING_r"(ComponentPose=@n22, Alpha=@n17, Translation="0,0,0", Rotation="0,0,0", Scale="0.9,1,1") {
    rotationMode=BMM_Additive
    rotationSpace=BCS_BoneSpace
    scaleMode=BMM_Replace
    scaleSpace=BCS_BoneSpace
    translationMode=BMM_Additive
    translationSpace=BCS_BoneSpace
    alphaBoolBlend {blendOption=Linear}
    alphaScaleBiasClamp {interpSpeedDecreasing=1, interpSpeedIncreasing=1}
    boneToModify {boneName=HairC_POSING_r}
    scale {x=0.9}
   }
   "Transform (Modify) Bone - Bone: HairB_POSING_r"(ComponentPose=@n23, Alpha=@n17, Translation="0,0,0", Rotation="0,0,0", Scale="0.9,1,1") {
    rotationMode=BMM_Additive
    rotationSpace=BCS_BoneSpace
    scaleMode=BMM_Replace
    scaleSpace=BCS_BoneSpace
    translationMode=BMM_Additive
    translationSpace=BCS_BoneSpace
    alphaBoolBlend {blendOption=Linear}
    alphaScaleBiasClamp {interpSpeedDecreasing=1, interpSpeedIncreasing=1}
    boneToModify {boneName=HairB_POSING_r}
    scale {x=0.9}
   }
   n24 = RigidBody(ComponentPose=Local To Component(LocalPose=Use cached pose 'Pose'()), Alpha=0.7) {
    bClampLinearTranslationLimitToRefPose=true
    bOverrideWorldGravity=true
    overridePhysicsAsset=&/F_MED_Rover_FaceAcc_Physics
    simulationSpace=BaseBoneSpace
    alphaBoolBlend {blendOption=Linear}
    baseBoneRef {boneName=head}
    componentAppliedLinearAccClamp=<3000 3000 3000>
    componentLinearAccScale=<0.3 0.3 0.3>
    componentLinearVelScale=<0.3 0.3 0.3>
    overrideWorldGravity {z=-980}
   }
   n25 = "Transform (Modify) Bone - Bone: HairC_POSING_r"!(ComponentPose=@n24, Alpha=1, Rotation="19.614,40.068,7.783") -> (Pose) {
    rotationMode=BMM_Replace
    rotationSpace=BCS_BoneSpace
    scaleSpace=BCS_BoneSpace
    translationSpace=BCS_BoneSpace
    alphaBoolBlend {blendOption=Linear}
    boneToModify {boneName=HairC_POSING_r}
    rotation=<rot 19.614 40.068 7.783>
   }
   n26 = "Transform (Modify) Bone - Bone: HairA_POSING_bk"!(ComponentPose=@n25, Alpha=1, Rotation="34.207,-17.591,-6.162") -> (Pose) {
    rotationMode=BMM_Replace
    rotationSpace=BCS_BoneSpace
    scaleSpace=BCS_BoneSpace
    translationSpace=BCS_BoneSpace
    alphaBoolBlend {blendOption=Linear}
    boneToModify {boneName=HairA_POSING_bk}
    rotation=<rot 34.207 -17.591 -6.162>
   }
   n27 = "Transform (Modify) Bone - Bone: HairC_POSING_l"!(ComponentPose=@n26, Alpha=1, Rotation="45.000,-0.041,0.053") -> (Pose) {
    rotationMode=BMM_Replace
    rotationSpace=BCS_BoneSpace
    scaleSpace=BCS_BoneSpace
    translationSpace=BCS_BoneSpace
    alphaBoolBlend {blendOption=Linear}
    boneToModify {boneName=HairC_POSING_l}
    rotation=<rot 45 -0.041 0.053>
   }
   n28 = Get HeadRotBackNoSkydive() -> (HeadRotBackNoSkydive:double)
   n29 = "Transform (Modify) Bone - Bone: HairC_POSING_l"!(
    ComponentPose=Local To Component(LocalPose=Use cached pose 'Pose'())
    Alpha=@n28
    Translation:Vector="0,0,0"
    Rotation="45,-45,0"
    Scale:Vector="1,1,1"
   ) -> (Pose) {
    rotationMode=BMM_Additive
    rotationSpace=BCS_BoneSpace
    scaleMode=BMM_Replace
    scaleSpace=BCS_BoneSpace
    translationMode=BMM_Additive
    translationSpace=BCS_BoneSpace
    alphaBoolBlend {blendOption=Linear}
    boneToModify {boneName=HairC_POSING_l}
    rotation {pitch=45, yaw=-45}
   }
   n30 = "Transform (Modify) Bone - Bone: HairB_POSING_l"(ComponentPose=@n29, Alpha=@n28, Translation="0,0,0", Rotation="45,-45,0", Scale="1,1,1") {
    rotationMode=BMM_Additive
    rotationSpace=BCS_BoneSpace
    scaleMode=BMM_Replace
    scaleSpace=BCS_BoneSpace
    translationMode=BMM_Additive
    translationSpace=BCS_BoneSpace
    alphaBoolBlend {blendOption=Linear}
    boneToModify {boneName=HairB_POSING_l}
    rotation {pitch=45, yaw=-45}
   }
   n31 = "Transform (Modify) Bone - Bone: HairA_POSING_bk"!(ComponentPose=@n30, Alpha=@n28, Translation:Vector="0,0,0", Rotation="45,0,0", Scale:Vector="1,1,1") -> (Pose) {
    rotationMode=BMM_Additive
    rotationSpace=BCS_BoneSpace
    scaleMode=BMM_Replace
    scaleSpace=BCS_BoneSpace
    translationMode=BMM_Additive
    translationSpace=BCS_BoneSpace
    alphaBoolBlend {blendOption=Linear}
    boneToModify {boneName=HairA_POSING_bk}
    rotation {pitch=45}
   }
   n32 = "Transform (Modify) Bone - Bone: HairA_POSING_l"(ComponentPose=@n31, Alpha=@n28, Translation="0,0,0", Rotation="45,-45,0", Scale="1,1,1") {
    rotationMode=BMM_Additive
    rotationSpace=BCS_BoneSpace
    scaleMode=BMM_Replace
    scaleSpace=BCS_BoneSpace
    translationMode=BMM_Additive
    translationSpace=BCS_BoneSpace
    alphaBoolBlend {blendOption=Linear}
    boneToModify {boneName=HairA_POSING_l}
    rotation {pitch=45, yaw=-45}
   }
   n33 = "Transform (Modify) Bone - Bone: HairA_POSING_r"(ComponentPose=@n32, Alpha=@n28, Translation="0,0,0", Rotation="45,0,0", Scale="1,1,1") {
    rotationMode=BMM_Additive
    rotationSpace=BCS_BoneSpace
    scaleMode=BMM_Replace
    scaleSpace=BCS_BoneSpace
    translationMode=BMM_Additive
    translationSpace=BCS_BoneSpace
    alphaBoolBlend {blendOption=Linear}
    boneToModify {boneName=HairA_POSING_r}
    rotation {pitch=45}
   }
   n34 = "Transform (Modify) Bone - Bone: HairC_POSING_r"!(ComponentPose=@n33, Alpha=@n28, Translation:Vector="0,0,0", Rotation="45,0,0", Scale:Vector="1,1,1") -> (Pose) {
    rotationMode=BMM_Additive
    rotationSpace=BCS_BoneSpace
    scaleMode=BMM_Replace
    scaleSpace=BCS_BoneSpace
    translationMode=BMM_Additive
    translationSpace=BCS_BoneSpace
    alphaBoolBlend {blendOption=Linear}
    boneToModify {boneName=HairC_POSING_r}
    rotation {pitch=45}
   }
   n35 = "Transform (Modify) Bone - Bone: HairB_POSING_r"(ComponentPose=@n34, Alpha=@n28, Translation="0,0,0", Rotation="45,0,0", Scale="1,1,1") {
    rotationMode=BMM_Additive
    rotationSpace=BCS_BoneSpace
    scaleMode=BMM_Replace
    scaleSpace=BCS_BoneSpace
    translationMode=BMM_Additive
    translationSpace=BCS_BoneSpace
    alphaBoolBlend {blendOption=Linear}
    boneToModify {boneName=HairB_POSING_r}
    rotation {pitch=45}
   }
   n36 = "Transform (Modify) Bone - Bone: HairA_POSING_r"(ComponentPose=@n35, Alpha=Get Is In Front End() -> (Is In Front End:double), Translation="-0.5,0,0", Rotation="45,0,0", Scale="1,1,1") {
    rotationMode=BMM_Additive
    rotationSpace=BCS_BoneSpace
    scaleMode=BMM_Replace
    scaleSpace=BCS_BoneSpace
    translationMode=BMM_Additive
    translationSpace=BCS_BoneSpace
    alphaBoolBlend {blendOption=Linear}
    boneToModify {boneName=HairA_POSING_r}
    rotation {pitch=45}
    translation {x=-0.5}
   }
   n37 = Get IsCustomEmotePlaying() -> (IsCustomEmotePlaying:double)
   n38 = "Transform (Modify) Bone - Bone: HairC_POSING_l"(ComponentPose=@n36, Alpha=@n37, Translation="0,0,0", Rotation="0,0,0", Scale="0.9,1,1") {
    rotationMode=BMM_Additive
    rotationSpace=BCS_BoneSpace
    scaleMode=BMM_Replace
    scaleSpace=BCS_BoneSpace
    translationMode=BMM_Additive
    translationSpace=BCS_BoneSpace
    alphaBoolBlend {blendOption=Linear}
    alphaScaleBiasClamp {interpSpeedDecreasing=1, interpSpeedIncreasing=1}
    boneToModify {boneName=HairC_POSING_l}
    scale {x=0.9}
   }
   n39 = "Transform (Modify) Bone - Bone: HairB_POSING_l"(ComponentPose=@n38, Alpha=@n37, Translation="0,0,0", Rotation="0,0,0", Scale="0.8,1,1") {
    rotationMode=BMM_Additive
    rotationSpace=BCS_BoneSpace
    scaleMode=BMM_Replace
    scaleSpace=BCS_BoneSpace
    translationMode=BMM_Additive
    translationSpace=BCS_BoneSpace
    alphaBoolBlend {blendOption=Linear}
    alphaScaleBiasClamp {interpSpeedDecreasing=1, interpSpeedIncreasing=1}
    boneToModify {boneName=HairB_POSING_l}
    scale {x=0.8}
   }
   n40 = "Transform (Modify) Bone - Bone: HairA_POSING_bk"(ComponentPose=@n39, Alpha=@n37, Translation="0,0,0", Rotation="0,0,0", Scale="0.8,1,1") {
    rotationMode=BMM_Additive
    rotationSpace=BCS_BoneSpace
    scaleMode=BMM_Replace
    scaleSpace=BCS_BoneSpace
    translationMode=BMM_Additive
    translationSpace=BCS_BoneSpace
    alphaBoolBlend {blendOption=Linear}
    alphaScaleBiasClamp {interpSpeedDecreasing=1, interpSpeedIncreasing=1}
    boneToModify {boneName=HairA_POSING_bk}
    scale {x=0.8}
   }
   n41 = "Transform (Modify) Bone - Bone: HairA_POSING_l"(ComponentPose=@n40, Alpha=@n37, Translation="0,0,0", Rotation="0,0,0", Scale="0.8,1,1") {
    rotationMode=BMM_Additive
    rotationSpace=BCS_BoneSpace
    scaleMode=BMM_Replace
    scaleSpace=BCS_BoneSpace
    translationMode=BMM_Additive
    translationSpace=BCS_BoneSpace
    alphaBoolBlend {blendOption=Linear}
    alphaScaleBiasClamp {interpSpeedDecreasing=1, interpSpeedIncreasing=1}
    boneToModify {boneName=HairA_POSING_l}
    scale {x=0.8}
   }
   n42 = "Transform (Modify) Bone - Bone: HairA_POSING_r"(ComponentPose=@n41, Alpha=@n37, Translation="0,0,0", Rotation="0,0,0", Scale="0.8,1,1") {
    rotationMode=BMM_Additive
    rotationSpace=BCS_BoneSpace
    scaleMode=BMM_Replace
    scaleSpace=BCS_BoneSpace
    translationMode=BMM_Additive
    translationSpace=BCS_BoneSpace
    alphaBoolBlend {blendOption=Linear}
    alphaScaleBiasClamp {interpSpeedDecreasing=1, interpSpeedIncreasing=1}
    boneToModify {boneName=HairA_POSING_r}
    scale {x=0.8}
   }
   n43 = "Transform (Modify) Bone - Bone: HairC_POSING_r"(ComponentPose=@n42, Alpha=@n37, Translation="0,0,0", Rotation="0,0,0", Scale="0.8,1,1") {
    rotationMode=BMM_Additive
    rotationSpace=BCS_BoneSpace
    scaleMode=BMM_Replace
    scaleSpace=BCS_BoneSpace
    translationMode=BMM_Additive
    translationSpace=BCS_BoneSpace
    alphaBoolBlend {blendOption=Linear}
    alphaScaleBiasClamp {interpSpeedDecreasing=1, interpSpeedIncreasing=1}
    boneToModify {boneName=HairC_POSING_r}
    scale {x=0.8}
   }
   n44 = "Transform (Modify) Bone - Bone: HairB_POSING_r"(ComponentPose=@n43, Alpha=@n37, Translation="0,0,0", Rotation="0,0,0", Scale="0.8,1,1") {
    rotationMode=BMM_Additive
    rotationSpace=BCS_BoneSpace
    scaleMode=BMM_Replace
    scaleSpace=BCS_BoneSpace
    translationMode=BMM_Additive
    translationSpace=BCS_BoneSpace
    alphaBoolBlend {blendOption=Linear}
    alphaScaleBiasClamp {interpSpeedDecreasing=1, interpSpeedIncreasing=1}
    boneToModify {boneName=HairB_POSING_r}
    scale {x=0.8}
   }
   n45 = RigidBody(ComponentPose=@n44, Alpha=Get NormalPhysics() -> (NormalPhysics:double)) {
    bClampLinearTranslationLimitToRefPose=true
    bOverrideWorldGravity=true
    overridePhysicsAsset=&F_MED_Rover_FaceAcc_Physics
    simulationSpace=BaseBoneSpace
    alphaBoolBlend {blendOption=Linear}
    baseBoneRef {boneName=head}
    componentAppliedLinearAccClamp=<3000 3000 3000>
    componentLinearAccScale=<0.3 0.3 0.3>
    componentLinearVelScale=<0.3 0.3 0.3>
    overrideWorldGravity {z=-980}
   }
   n46 = RigidBody(ComponentPose=@n45, Alpha=Get IsCustomEmotePlaying()) {
    bClampLinearTranslationLimitToRefPose=true
    bOverrideWorldGravity=true
    overridePhysicsAsset=&/F_MED_Rover_FaceAcc_PhysicsDanceEmote
    simulationSpace=BaseBoneSpace
    alphaBoolBlend {blendOption=Linear}
    baseBoneRef {boneName=head}
    componentAppliedLinearAccClamp=<3000 3000 3000>
    componentLinearAccScale=<0.3 0.3 0.3>
    componentLinearVelScale=<0.3 0.3 0.3>
    overrideWorldGravity {z=-980}
   }
   n47 = Get Is Skydive Floating() -> (Is Skydive Floating:double)
   n48 = "Transform (Modify) Bone - Bone: HairA_POSING_bk"(ComponentPose=@n46, Alpha=@n47, Translation="0,0,0", Rotation="40,0,0", Scale="0.85,1,1") {
    rotationMode=BMM_Additive
    rotationSpace=BCS_BoneSpace
    scaleMode=BMM_Replace
    scaleSpace=BCS_BoneSpace
    translationMode=BMM_Additive
    translationSpace=BCS_BoneSpace
    alphaBoolBlend {blendOption=Linear}
    alphaScaleBiasClamp {bInterpResult=true}
    boneToModify {boneName=HairA_POSING_bk}
    rotation {pitch=40}
    scale {x=0.85}
   }
   n49 = "Transform (Modify) Bone - Bone: HairC_POSING_r"(ComponentPose=@n48, Alpha=@n47, Translation="0,5,0", Rotation="40,30,0", Scale="1,1,1") {
    rotationMode=BMM_Additive
    rotationSpace=BCS_BoneSpace
    scaleMode=BMM_Replace
    scaleSpace=BCS_BoneSpace
    translationMode=BMM_Additive
    translationSpace=BCS_BoneSpace
    alphaBoolBlend {blendOption=Linear}
    alphaScaleBiasClamp {bInterpResult=true}
    boneToModify {boneName=HairC_POSING_r}
    rotation {pitch=40, yaw=30}
    translation {y=5}
   }
   n50 = "Transform (Modify) Bone - Bone: HairC_POSING_l"(ComponentPose=@n49, Alpha=@n47, Translation="-5,0,0", Rotation="40,-30,0", Scale="1,1,1") {
    rotationMode=BMM_Additive
    rotationSpace=BCS_BoneSpace
    scaleMode=BMM_Replace
    scaleSpace=BCS_BoneSpace
    translationMode=BMM_Additive
    translationSpace=BCS_BoneSpace
    alphaBoolBlend {blendOption=Linear}
    alphaScaleBiasClamp {bInterpResult=true}
    boneToModify {boneName=HairC_POSING_l}
    rotation {pitch=40, yaw=-30}
    translation {x=-5}
   }
   n51 = "Transform (Modify) Bone - Bone: HairB_POSING_r"(ComponentPose=@n50, Alpha=@n47, Translation="5,0,0", Rotation="20,-20,0", Scale="1,1,1") {
    rotationMode=BMM_Additive
    rotationSpace=BCS_BoneSpace
    scaleMode=BMM_Replace
    scaleSpace=BCS_BoneSpace
    translationMode=BMM_Additive
    translationSpace=BCS_BoneSpace
    alphaBoolBlend {blendOption=Linear}
    alphaScaleBiasClamp {bInterpResult=true}
    boneToModify {boneName=HairB_POSING_r}
    rotation {pitch=20, yaw=-20}
    translation {x=5}
   }
   n52 = "Transform (Modify) Bone - Bone: HairB_POSING_l"(ComponentPose=@n51, Alpha=@n47, Translation="-5,0,0", Rotation="20,-20,0", Scale="1,1,1") {
    rotationMode=BMM_Additive
    rotationSpace=BCS_BoneSpace
    scaleMode=BMM_Replace
    scaleSpace=BCS_BoneSpace
    translationMode=BMM_Additive
    translationSpace=BCS_BoneSpace
    alphaBoolBlend {blendOption=Linear}
    alphaScaleBiasClamp {bInterpResult=true}
    boneToModify {boneName=HairB_POSING_l}
    rotation {pitch=20, yaw=-20}
    translation {x=-5}
   }
   n53 = Get Is Skydive Dive Mode() -> (Is Skydive Dive Mode:double)
   n54 = "Transform (Modify) Bone - Bone: HairA_POSING_bk"(ComponentPose=@n52, Alpha=@n53, Translation="0,0,0", Rotation="20,0,0", Scale="1,1,1") {
    rotationMode=BMM_Additive
    rotationSpace=BCS_BoneSpace
    scaleMode=BMM_Replace
    scaleSpace=BCS_BoneSpace
    translationMode=BMM_Additive
    translationSpace=BCS_BoneSpace
    alphaBoolBlend {blendOption=Linear}
    boneToModify {boneName=HairA_POSING_bk}
    rotation {pitch=20}
   }
   n55 = "Transform (Modify) Bone - Bone: HairC_POSING_r"(ComponentPose=@n54, Alpha=@n53, Translation="0,0,0", Rotation="20,-10,0", Scale="1,1,1") {
    rotationMode=BMM_Additive
    rotationSpace=BCS_BoneSpace
    scaleMode=BMM_Replace
    scaleSpace=BCS_BoneSpace
    translationMode=BMM_Additive
    translationSpace=BCS_BoneSpace
    alphaBoolBlend {blendOption=Linear}
    boneToModify {boneName=HairC_POSING_r}
    rotation {pitch=20, yaw=-10}
   }
   n56 = "Transform (Modify) Bone - Bone: HairC_POSING_l"(ComponentPose=@n55, Alpha=@n53, Translation="0,0,0", Rotation="20,-10,0", Scale="1,1,1") {
    rotationMode=BMM_Additive
    rotationSpace=BCS_BoneSpace
    scaleMode=BMM_Replace
    scaleSpace=BCS_BoneSpace
    translationMode=BMM_Additive
    translationSpace=BCS_BoneSpace
    alphaBoolBlend {blendOption=Linear}
    boneToModify {boneName=HairC_POSING_l}
    rotation {pitch=20, yaw=-10}
   }
   n57 = Get skydiveGlider() -> (skydiveGlider:double)
   n58 = "Transform (Modify) Bone - Bone: HairA_POSING_bk"(ComponentPose=@n56, Alpha=@n57, Translation="0,0,0", Rotation="10,0,0", Scale="1,1,1") {
    rotationMode=BMM_Additive
    rotationSpace=BCS_BoneSpace
    scaleMode=BMM_Replace
    scaleSpace=BCS_BoneSpace
    translationMode=BMM_Additive
    translationSpace=BCS_BoneSpace
    alphaBoolBlend {blendOption=Linear}
    boneToModify {boneName=HairA_POSING_bk}
    rotation {pitch=10}
   }
   n59 = "Transform (Modify) Bone - Bone: HairC_POSING_r"(ComponentPose=@n58, Alpha=@n57, Translation="0,0,0", Rotation="10,-5,0", Scale="1,1,1") {
    rotationMode=BMM_Additive
    rotationSpace=BCS_BoneSpace
    scaleMode=BMM_Replace
    scaleSpace=BCS_BoneSpace
    translationMode=BMM_Additive
    translationSpace=BCS_BoneSpace
    alphaBoolBlend {blendOption=Linear}
    boneToModify {boneName=HairC_POSING_r}
    rotation {pitch=10, yaw=-5}
   }
   n60 = "Transform (Modify) Bone - Bone: HairC_POSING_l"(ComponentPose=@n59, Alpha=@n57, Translation="0,0,0", Rotation="10,-5,0", Scale="1,1,1") {
    rotationMode=BMM_Additive
    rotationSpace=BCS_BoneSpace
    scaleMode=BMM_Replace
    scaleSpace=BCS_BoneSpace
    translationMode=BMM_Additive
    translationSpace=BCS_BoneSpace
    alphaBoolBlend {blendOption=Linear}
    boneToModify {boneName=HairC_POSING_l}
    rotation {pitch=10, yaw=-5}
   }
   n61 = RigidBody(ComponentPose=@n60, Alpha=Get Is Sky Diving() -> (Is Sky Diving:double)) {
    bClampLinearTranslationLimitToRefPose=true
    bOverrideWorldGravity=true
    overridePhysicsAsset=&/F_MED_Rover_FaceAcc_PhysicsSkyDive
    simulationSpace=BaseBoneSpace
    alphaBoolBlend {blendOption=Linear}
    baseBoneRef {boneName=head}
    componentAppliedLinearAccClamp=<3000 3000 3000>
    componentLinearAccScale=<0.4 0.3 0.3>
    componentLinearVelScale=<0.4 0.3 0.3>
    overrideWorldGravity {z=-10}
   }
   n62 = Get HeadRotDelay() -> (HeadRotDelay:double)
   n63 = "Transform (Modify) Bone - Bone: HairC_POSING_l"(ComponentPose=@n61, Alpha=@n62, Translation="-4,0,-4", Rotation="0,0,0", Scale="1,1,1") {
    rotationMode=BMM_Additive
    rotationSpace=BCS_BoneSpace
    scaleMode=BMM_Replace
    scaleSpace=BCS_BoneSpace
    translationMode=BMM_Additive
    translationSpace=BCS_BoneSpace
    alphaBoolBlend {blendOption=Linear}
    alphaScaleBiasClamp {bInterpResult=true, interpSpeedIncreasing=1}
    boneToModify {boneName=HairC_POSING_l}
    translation {x=-4, z=-4}
   }
   n64 = "Transform (Modify) Bone - Bone: HairA_POSING_bk"(ComponentPose=@n63, Alpha=@n62, Translation="-4,0,-4", Rotation="0,0,0", Scale="1,1,1") {
    rotationMode=BMM_Additive
    rotationSpace=BCS_BoneSpace
    scaleMode=BMM_Replace
    scaleSpace=BCS_BoneSpace
    translationMode=BMM_Additive
    translationSpace=BCS_BoneSpace
    alphaBoolBlend {blendOption=Linear}
    alphaScaleBiasClamp {bInterpResult=true, interpSpeedIncreasing=1}
    boneToModify {boneName=HairA_POSING_bk}
    translation {x=-4, z=-4}
   }
   n65 = "Transform (Modify) Bone - Bone: HairC_POSING_r"(ComponentPose=@n64, Alpha=@n62, Translation="-4,0,-4", Rotation="0,0,0", Scale="1,1,1") {
    rotationMode=BMM_Additive
    rotationSpace=BCS_BoneSpace
    scaleMode=BMM_Replace
    scaleSpace=BCS_BoneSpace
    translationMode=BMM_Additive
    translationSpace=BCS_BoneSpace
    alphaBoolBlend {blendOption=Linear}
    alphaScaleBiasClamp {bInterpResult=true, interpSpeedIncreasing=1}
    boneToModify {boneName=HairC_POSING_r}
    translation {x=-4, z=-4}
   }
   n66 = "Transform (Modify) Bone - Bone: HairA_POSING_r"(ComponentPose=@n65, Alpha=1, Translation="-4,0,0", Rotation="0,0,0", Scale="1,1,1") {
    rotationMode=BMM_Additive
    rotationSpace=BCS_BoneSpace
    scaleMode=BMM_Replace
    scaleSpace=BCS_BoneSpace
    translationMode=BMM_Additive
    translationSpace=BCS_BoneSpace
    alphaBoolBlend {blendOption=Linear}
    alphaScaleBiasClamp {bInterpResult=true, interpSpeedIncreasing=1}
    boneToModify {boneName=HairA_POSING_r}
    translation {x=-4}
   }
   n67 = Get Is In Front End()
   n68 = "Transform (Modify) Bone - Bone: HairA3_r"(ComponentPose=@n66, Alpha=@n67, Translation="2,-2,0", Rotation="-30,0,0", Scale="1,1,1") {
    rotationMode=BMM_Additive
    rotationSpace=BCS_BoneSpace
    scaleMode=BMM_Replace
    scaleSpace=BCS_BoneSpace
    translationMode=BMM_Additive
    translationSpace=BCS_BoneSpace
    alphaBoolBlend {blendOption=Linear}
    boneToModify {boneName=HairA3_r}
    rotation {pitch=-30}
    translation=<2 -2>
   }
   n69 = "Transform (Modify) Bone - Bone: HairA_POSING_r"(ComponentPose=@n68, Alpha=@n67, Translation="-1,0,0", Rotation="0,0,0", Scale="1,1,1") {
    rotationMode=BMM_Additive
    rotationSpace=BCS_BoneSpace
    scaleMode=BMM_Replace
    scaleSpace=BCS_BoneSpace
    translationMode=BMM_Additive
    translationSpace=BCS_BoneSpace
    alphaBoolBlend {blendOption=Linear}
    boneToModify {boneName=HairA_POSING_r}
    translation {x=-1}
   }
   n70 = "Transform (Modify) Bone - Bone: HairB3_r"(ComponentPose=@n69, Alpha=1, Translation="0,0,0", Rotation="0,-20,0", Scale="1,1,1") {
    rotationMode=BMM_Additive
    rotationSpace=BCS_BoneSpace
    scaleMode=BMM_Replace
    scaleSpace=BCS_BoneSpace
    translationMode=BMM_Additive
    translationSpace=BCS_BoneSpace
    alphaBoolBlend {blendOption=Linear}
    boneToModify {boneName=HairB3_r}
    rotation {yaw=-20}
   }
   Output Pose(Result=Blend Poses by bool(BlendPose_0=Component To Local(ComponentPose=@n27), BlendPose_1=Component To Local(ComponentPose=@n70), bActiveValue=Get bIsDBNO()) {blendType=Linear}) {layerGroup=None}
   n71 = Get Body_Mesh()
   Save cached pose 'Pose'(Pose=Layered blend per bone(BasePose=Copy Pose From Mesh() {bCopyCurves=true, bUseAttachedParent=true}, BlendPoses_0=Copy Pose From Mesh(SourceMeshComponent=@n71)) {layerSetup [{branchFilters [{boneName=upperarm_l}, {boneName=upperarm_r}]}]}) {cacheName=Pose}
   Get HeadRotBackNoSkydiveEmote() -> (HeadRotBackNoSkydiveEmote:double)
   Get HeadRotBackNoSkydive()
   Get IsCustomEmotePlaying()
   Get HeadRotBackNoSkydiveEmote()
  ]}
 EventGraph {nodes [
   Event BlueprintUpdateAnimation()
   > n1 = Sequence!(execute) -> (
    then_0
    then_1
    then_2:Exec
    then_3:Exec
    then_4:Exec
    then_5:Exec
    then_6:Exec
    then_7:Exec
    then_8:Exec
    then_9:Exec
    then_10:Exec
    then_11:Exec
    then_12:Exec
    then_13:Exec
    then_14:Exec
    then_15:Exec
    then_16:Exec
   )
   n2 = Get body Mesh() -> (body Mesh:Skeletal Mesh Component Object Reference)
   > n3 = Set head Rot Back(head Rot Back:double=GetAlphaFromReferenceTransformDelta(SkeletalMeshComponent=@n2, BoneName=head, BaseName=spine_05, InputRange=(X=-15,Y=0), OutputRange=(X=1,Y=0), TwistAxis="1,0,0")) -> (Output_Get:double)
   n4 = Set using Pistol(execute=@n11.then, using Pistol:bool=true) -> (Output_Get:bool)
   n5 = Branch(execute=@n3.then, Condition=@n4)
   > Set head Rot Pistol On(head Rot Pistol On:double=@n3) -> (Output_Get:double)
   Set head Rot Pistol On(execute=@n5.else, head Rot Pistol On=0)
   Set body Mesh(execute=@n1.then_1, body Mesh:Skeletal Mesh Component Object Reference=GetOwnerSkeletalMeshComponent()) -> (Output_Get:Skeletal Mesh Component Object Reference)
   Set HeadRotWithBling(
    execute=@n1.then_10
    HeadRotWithBling:double=float * float(A=Get head Rot Back() -> (head Rot Back:double), B=Get Is Wearing Backbling() -> (Is Wearing Backbling:double)) {is=title K2Node_PromotableOperator}
   ) -> (Output_Get:double)
   n6 = NOT Boolean(A=Get bIsCrouching())
   Set HeadRotNoCrouch(execute=@n1.then_11, HeadRotNoCrouch:double=float * float(A=Get HeadRotBackNoSkydive(), B="To Float (Boolean)"(InBool=@n6))) -> (Output_Get:double)
   Set HeadRotDelay(
    execute=@n1.then_12
    HeadRotDelay:double=GetAlphaFromReferenceTransformDelta(SkeletalMeshComponent=@n2, BoneName=head, BaseName=spine_05, InputRange=(X=-45,Y=-15), OutputRange=(X=1,Y=0), TwistAxis="1,0,0")
   ) -> (Output_Get:double)
   n7 = GetAnimInstance(self=GetOwnerSkeletalMeshComponent())
   Is Valid(exec=@n1.then_13, InputObject=@n7)
   > n8 = Set IsCustomEmotePlaying(IsCustomEmotePlaying:double=SelectFloat(A=1, B=0, bPickA=Montage_IsPlaying(self=@n7, Montage=/Game/Animation/Game/MainPlayer/Montages/Emotes/Emote_DanceMoves))) -> (Output_Get:double)
   > Set CustomEmoteNotPlaying(CustomEmoteNotPlaying:double="To Float (Boolean)"(InBool="float < float"(A=@n8, B=0.5))) -> (Output_Get:double)
   Set NormalPhysics(
    execute=@n1.then_14
    NormalPhysics:double=float * float(A=Get CustomEmoteNotPlaying() -> (CustomEmoteNotPlaying:double), B=Get Is Not Sky Diving() -> (Is Not Sky Diving:double))
   ) -> (Output_Get:double)
   Set HeadRotBackNoSkydiveEmote(execute=@n1.then_15, HeadRotBackNoSkydiveEmote:double=float * float(A=Get CustomEmoteNotPlaying(), B=Get HeadRotBackNoSkydive())) -> (Output_Get:double)
   Set CusomEmoteFrontend(execute=@n1.then_16, CusomEmoteFrontend:double=float * float(A=Get Is In Front End(), B=Get IsCustomEmotePlaying())) -> (Output_Get:double)
   n9 = Cast To FortPlayerAnimInstance(execute=@n1.then_2, Object=GetAnimInstance(self=GetOwnerSkeletalMeshComponent()))
   > Set ar_firingPosition(ar_firingPosition:double=0) -> (Output_Get:double)
   > n10 = Switch on EFortWeaponCoreAnimation!(execute, Selection:EFortWeaponCoreAnimation Enum=Get WeaponCoreAnim(self=@n9)) -> (
    Melee:Exec
    Pistol:Exec
    Shotgun:Exec
    PaperBlueprint:Exec
    Rifle:Exec
    MeleeOneHand:Exec
    MachinePistol:Exec
    RocketLauncher:Exec
    GrenadeLauncher:Exec
    GoingCommando:Exec
    AssaultRifle:Exec
    TacticalShotgun:Exec
    SniperRifle:Exec
    TrapPlacement:Exec
    ShoulderLauncher:Exec
    AbilityDecoTool:Exec
    Crossbow:Exec
    C4:Exec
    RemoteControl:Exec
    DualWield:Exec
    AR_BullPup:Exec
    AR_ForwardGrip:Exec
    MedPackPaddles:Exec
    SMG_P90:Exec
    AR_DrumGun:Exec
    Consumable_Small:Exec
    Consumable_Large:Exec
    Balloon:Exec
    MountedTurret:Exec
    CreativeTool:Exec
    ExplosiveBow:Exec
    AshtonIndigo:Exec
    AshtonChicago:Exec
    MeleeDualWield:Exec
    Unarmed:Exec
   )
   n11 = Branch(execute=@n10.Pistol, Condition=Get bIsTargeting())
   Set using Pistol(execute=@n11.else, using Pistol=false)
   Set Is In Front End(execute=@n1.then_3, Is In Front End:double="To Float (Boolean)"(InBool=Get bIsInFrontEnd())) -> (Output_Get:double)
   n12 = Get bIsSkyDiving()
   Set Is Not Sky Diving(execute=@n1.then_4, Is Not Sky Diving:double="To Float (Boolean)"(InBool=NOT Boolean(A=@n12))) -> (Output_Get:double)
   Set Is Sky Diving(execute=@n1.then_5, Is Sky Diving:double="To Float (Boolean)"(InBool=Get bIsSkyDiving())) -> (Output_Get:double)
   n13 = NOT Boolean(A=Get bIsFalling())
   n14 = NOT Boolean(A=Get bIsJumping())
   Set HeadRotBackNoSkydive(
    execute=@n1.then_6
    HeadRotBackNoSkydive:double=float * float!(A=Get head Rot Back(), B=Get Is Not Sky Diving(), C:double="To Float (Boolean)"(InBool=@n14), D:double="To Float (Boolean)"(InBool=@n13)) -> (ReturnValue)
   ) -> (Output_Get:double)
   Set Is Skydive Floating(execute=@n1.then_7, Is Skydive Floating:double="To Float (Boolean)"(InBool=Get bIsSkydiveFloating())) -> (Output_Get:double)
   Set Is Skydive Dive Mode(execute=@n1.then_8, Is Skydive Dive Mode:double="To Float (Boolean)"(InBool=Get bIsSkydiveDiveMode())) -> (Output_Get:double)
   n15 = NOT Boolean(A=Get bIsSkyDiving())
   n16 = NOT Boolean(A=Get bIsSkydiveFloating())
   n17 = Get bIsSkydiveDiveMode()
   Set skydiveGlider(execute=@n1.then_9, skydiveGlider:double="To Float (Boolean)"(InBool=AND Boolean!(A=@n15, B=@n16, C:bool=@n17) -> (ReturnValue))) -> (Output_Get:double)
   Event BlueprintInitializeAnimation()
   > Set Body_Mesh(Body_Mesh=GetOwnerSkeletalMeshComponent())
   Event OnCharacterCustomizationLoaded()
   > Do Once()
   n18 = CheckForBackbling()
   > Set Is Wearing Backbling(Is Wearing Backbling:double="To Float (Boolean)"(InBool=OR Boolean(A=@n18, B=@n18.bIsBackblingLong))) -> (Output_Get:double)
   Get bIsJumping()
  ]}
]
variables [
 ar_firingPosition {type=number}
 body Mesh {type=SkeletalMeshComponent}
 Body_Mesh {type=SkeletalMeshComponent}
 CusomEmoteFrontend {type=number}
 CustomEmoteNotPlaying {type=number}
 DelayedHeadRot {type=number}
 head Rot Back {type=number}
 head Rot Pistol On {type=number}
 HeadRotBackNoSkydive {type=number}
 HeadRotBackNoSkydiveEmote {type=number}
 HeadRotDelay {type=number}
 HeadRotNoCrouch {type=number}
 HeadRotWithBling {type=number}
 Is In Front End {type=number}
 Is Jumping {type=boolean}
 Is Not Sky Diving {type=number}
 Is Sky Diving {type=number}
 Is Skydive Dive Mode {type=number}
 Is Skydive Floating {type=number}
 Is Wearing Backbling {type=number}
 IsCustomEmotePlaying {type=number}
 isNotCustomEmote {type=number}
 NormalPhysics {type=number}
 skydiveGlider {type=number}
 using Pistol {type=boolean}
]
